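# Idle Niagara effect for the CadetFrog character body: a few looping fireflies that hover around it with glowing meshes, flickering sprites and a light.
# Curl noise jitters each firefly while a pendulum constraint tethers it to a vertically bobbing pivot, varied by particle index.
# A system-level intensity value scales glow and light, with a time-of-day input among dependencies.
# A second burst's spawn group feeds light visibility, and light brightness differs between front end and gameplay.
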NiagaraSystem /BRCosmetics/Characters/Player/Male/Medium/Bodies/M_MED_CadetFrog/FX/NS_CadetFrog_Character_Idle
dependencies {
 usedDataInterfaces [&NiagaraDataInterfaceCurve]
 usedDynamicInputs [&FNTimeOfDayScalar]
 usedModules [
  &self:SystemSpawnScript.SystemScriptSource.SystemScriptGraph.NiagaraNodeAssignment_0.SetVariables_#1
  &SystemState
  &self:SystemSpawnScript.SystemScriptSource.SystemScriptGraph.NiagaraNodeAssignment_2.SetVariables_#2
 ]
}
emitters {Fireflies {
  emitterUpdateScript [
   &EmitterState(
    Inactive Response=::NewEnumerator1
    Life Cycle Mode=::NewEnumerator1
    Loop Behavior=::NewEnumerator1
    Loop Duration Mode=::NewEnumerator1
    UseLoopDelay=false
   )
   &SpawnBurst_Instantaneous(
    Spawn Count=&RandomRangeInt(Minimum=3, Maximum=4)
    Use Spawn Probability=false
    UseLoopCountLimit=false
   )
   &SpawnBurst_Instantaneous(Spawn Group=1, Use Spawn Probability=false, UseLoopCountLimit=false) {module=SpawnBurst_Instantaneous001}
  ]
  particleSpawnScript [&InitializeParticle(
    Color Mode=::NewEnumerator0
    Mesh Scale Mode=::NewEnumerator1
    Mesh Uniform Scale Max=0.333
    Mesh Uniform Scale Min=0.266
    Sprite Size Mode=::NewEnumerator3
    Uniform Sprite Size=7
    UsePositionOffset=false
    Write Material Random=true
   )]
  particleUpdateScript [
   &ParticleState(
    Kill Particles When Lifetime Has Elapsed=false
    Let Infinitely Lived Particles Die When Emitter Deactivates=true
    Loop Particles Lifetime=true
   )
   /SetVariables(
    Particles.FireflyMeshColor:LinearColor=&Multiply_LinearColorByFloat(Linear Color=<rgba 5 4 2 1>, Float=$System.ScaleIntensity)
    Particles.GlowSpriteCameraOffset=0.01
    Particles.GlowSpriteColor:LinearColor=&MultiplyLinearColor(
     Linear Color=<rgba 1 0.7507 0.2522 0.5>
     Scale Factor=&MakeVector4_FromFloat(Float=&Multiply_Float(
       A=&Multiply_Float(A=0.5, B=$System.ScaleIntensity)
       B=&Waveform(
        "Waveform [1]"=::NewEnumerator0
        "[1] Frequency"=&RandomRangeFloat(Minimum=0.4, Maximum=0.8)
        "[1] Amplitude Min/Max"=<0.6 1>
       )
      ))
    )
   ) {isSetParameters=true}
   &DynamicMaterialParameters(
    Index 0 Param 1=3
    Index 0 Param 2=9
    Index 0 Param 3=1
    Module.Index0Param0Write=true
    Module.Index0Param1Write=true
    Module.Index0Param2Write=true
    Module.Index0Param3Write=false
   )
   &CurlNoiseForce(
    Noise Frequency=&RandomRangeFloat(Minimum=44, Maximum=77)
    Noise Strength=&RandomRangeFloat(Minimum=222, Maximum=444)
    Pan Noise=true
    Pan Noise Field=&MakeVector(X=0, Y=0, Z=2)
    Scale Curl Lookup By Engine Owner Scale=false
   )
   &PendulumSetup(
    Module.Enable Spring Driver=false
    Pendulum Length=2
    Pendulum Pivot:Vector3f=&AddVectorToPosition(
     Position=/Niagara/DynamicInputs/Helpers/SimulationPosition()
     Vector=&MakeVector(
      X=0
      Y=0
      Z=&FloatFromCurve(
       FloatCurve=DataInterface() {propertyValues {Curve {
          defaultValue=1
          keys [
           {interpMode=RCIM_Cubic, tangentMode=RCTM_User, time=0, value=-1}
           {interpMode=RCIM_Cubic, tangentMode=RCTM_User, time=0.5, value=1}
           {tangentMode=RCTM_User, time=1, value=-1}
          ]
         }}}
       Scale Curve=&FloatFromCurve(
        FloatCurve=DataInterface() {propertyValues {Curve {defaultValue=1, keys [{tangentMode=RCTM_User, time=0, value=-1}, {tangentMode=RCTM_User, time=1, value=1}]}}}
        CurveIndex=&ReturnNormalizedExecIndex()
        Scale Curve=2.5
       )
      )
     )
    )
    Pendulum Pivot Offset Coordinate Space:ENiagaraCoordinateSpace=::Local
    Rigid Pendulum=false
   )
   &SolveForcesAndVelocity(Clamp Velocity=false, Limit Acceleration=false)
   &PendulumConstraint()
   &CalculateAccurateVelocity()
   &FN_Light_Attributes(
    Color=<rgba 1 0.5333 0 1>
    Create Color Attribute=true
    Create Exponent Attribute=true
    Create Position Attribute=true
    Create Radius Attribute=true
    Enable Visibility Tag=true
    Exponent=7.5
    Position=&AddVectorToPosition(
     Position=&SimulationPosition()
     Vector=&MakeVector(
      X=0
      Y=0
      Z=&FloatFromCurve(
       FloatCurve=DataInterface() {propertyValues {Curve {
          defaultValue=1
          keys [
           {interpMode=RCIM_Cubic, tangentMode=RCTM_User, time=0, value=-1}
           {interpMode=RCIM_Cubic, tangentMode=RCTM_User, time=0.5, value=1}
           {tangentMode=RCTM_User, time=1, value=-1}
          ]
         }}}
       Scale Curve=&FloatFromCurve(
        FloatCurve=DataInterface() {propertyValues {Curve {defaultValue=1, keys [{tangentMode=RCTM_User, time=0, value=-1}, {tangentMode=RCTM_User, time=1, value=1}]}}}
        CurveIndex=&ReturnNormalizedExecIndex()
       )
      )
     )
    )
    Radius=100
    Scale Color=&Waveform(
     "Waveform [1]"=::NewEnumerator12
     "[1] Phase Scale"=$System.ScaleIntensity
     "[1] Frequency"=1
     "[1] Amplitude Min/Max"=<0 1>
     "[1] Scale"=&FN_FrontEndFloatModifier(Gameplay=7.5, Front End=15)
    )
    VisibilityTag:NiagaraInt32=&SpawnGroup()
   )
  ]
  rendererClasses [
   &NiagaraMeshRendererProperties
   &NiagaraSpriteRendererProperties
   &NiagaraLightRendererProperties
  ]
  renderers [
   {rendererClass=&NiagaraMeshRendererProperties, rendererIndex=0}
   {rendererClass=&NiagaraSpriteRendererProperties, rendererIndex=1}
   {rendererClass=&NiagaraMeshRendererProperties, rendererIndex=2}
   {rendererClass=&NiagaraSpriteRendererProperties, rendererIndex=3}
   {rendererClass=&NiagaraLightRendererProperties, rendererIndex=4}
  ]
 }}
summary {
 systemName=NS_CadetFrog_Character_Idle
 emitters [{
   emitterName=Fireflies
   rendererClasses [
    &NiagaraMeshRendererProperties
    &NiagaraSpriteRendererProperties
    &NiagaraLightRendererProperties
   ]
  }]
 userVariables [
  User.bIsFrontEnd {type=NiagaraBool}
  User.bIsFrontEndPreview {type=NiagaraBool}
 ]
}
