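# Physics asset for the front tube on the CadetFrog medium male player body, defining capsule bodies for the tube and the upper torso around it.
# Large capsules on spine_05 and both clavicles form the torso, with each clavicle constrained to the spine.
# The tube chains from spine_05 through a tiny base body with free swing and twist to a dynamic segment with narrow swing limits and locked twist.
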PhysicsAsset /BRCosmetics/Characters/Player/Male/Medium/Bodies/M_MED_CadetFrog/Meshes/M_MED_CadetFrog_FrontTube_Physics
bodies {
 base_tube_front_cr [{shapeType=Capsule, length=0.505, radius=0.505, shapeName=base_tube_front_cr_capsule, rotation=<rot 0 -90 0>}]
 clavicle_l [{shapeType=Capsule, length=3.7819, radius=10.8529, center=<6.7031 0.7988 -1.8526>, rotation=<rot 9.5591 -99.4399 -68.8289>}]
 clavicle_r [{shapeType=Capsule, length=3.7819, radius=10.8529, center=<-6.7031 -0.7988 1.8526>, rotation=<rot 9.5591 -99.4399 111.1711>}]
 dyn_tube_front_cr [{
   shapeType=Capsule
   length=9.4894
   radius=3.8656
   shapeName=dyn_tube_front_cr_capsule
   center=<7.0188 0.0018 0.4513>
   rotation=<rot -0.0001 -176.526 -86.9938>
  }]
 spine_05 [{shapeType=Capsule, length=10.1197, radius=13.726, center=<-3.7033 -0.2537 0>}]
}
constraints [
 {bone1Name=clavicle_l, bone2Name=spine_05, swing1LimitDegrees=45, swing2LimitDegrees=45, twistLimitDegrees=45}
 {bone1Name=clavicle_r, bone2Name=spine_05, swing1LimitDegrees=45, swing2LimitDegrees=45, twistLimitDegrees=45}
 {
  bone1Name=dyn_tube_front_cr
  bone2Name=base_tube_front_cr
  swing1LimitDegrees=20
  swing2LimitDegrees=20
  twistLimitDegrees=45
  twistMotion=Locked
 }
 {
  bone1Name=base_tube_front_cr
  bone2Name=spine_05
  swing1LimitDegrees=45
  swing1Motion=Free
  swing2LimitDegrees=45
  swing2Motion=Free
  twistLimitDegrees=45
  twistMotion=Free
 }
]
