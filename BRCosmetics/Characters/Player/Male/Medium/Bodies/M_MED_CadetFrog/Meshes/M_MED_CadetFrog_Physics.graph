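# Physics asset for the M_MED_CadetFrog player cosmetic body, covering the pelvis, both thighs and a jar bone chain, apparently for secondary motion of the jar.
# The jar chain runs pelvis, base_jar, offset_jar, dyn_jar, with free rotation on the first two joints and tight angular limits on the last.
# Tiny capsules serve base_jar and offset_jar, while larger ones shape dyn_jar, the pelvis and the pelvis-constrained thighs.
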PhysicsAsset /BRCosmetics/Characters/Player/Male/Medium/Bodies/M_MED_CadetFrog/Meshes/M_MED_CadetFrog_Physics
bodies {
 base_jar [{shapeType=Capsule, length=0.505, radius=0.505, shapeName=base_jar_capsule, rotation=<rot 0 -90 0>}]
 dyn_jar [{
   shapeType=Capsule
   length=8.4207
   radius=5.1419
   shapeName=dyn_jar_capsule
   center=<4.6766 0.1109 -1.5486>
   rotation=<rot -25.2184 -91.545 16.8533>
  }]
 offset_jar [{shapeType=Capsule, length=0.505, radius=0.505, shapeName=offset_jar_capsule, rotation=<rot 0 -90 0>}]
 pelvis [{shapeType=Capsule, length=6.4403, radius=10.0294, center=<1.8097 -1.2565 4.5874>, rotation=<rot -90 -0.0001 0>}]
 thigh_l [{shapeType=Capsule, length=19.6058, radius=9.0238, center=<-20.007 -0.7076 -0.2925>, rotation=<rot 0 -89.0074 -88.8421>}]
 thigh_r [{shapeType=Capsule, length=19.6058, radius=9.2454, center=<14.0023 1.282 -1.5106>, rotation=<rot 0.1009 -94.0066 -88.8465>}]
}
constraints [
 {bone1Name=thigh_l, bone2Name=pelvis, swing1LimitDegrees=45, swing2LimitDegrees=45, twistLimitDegrees=45}
 {bone1Name=thigh_r, bone2Name=pelvis, swing1LimitDegrees=45, swing2LimitDegrees=45, twistLimitDegrees=45}
 {
  bone1Name=offset_jar
  bone2Name=base_jar
  swing1LimitDegrees=45
  swing1Motion=Free
  swing2LimitDegrees=45
  swing2Motion=Free
  twistLimitDegrees=45
  twistMotion=Free
 }
 {bone1Name=dyn_jar, bone2Name=offset_jar, swing1LimitDegrees=5, swing2LimitDegrees=10, twistLimitDegrees=5}
 {
  bone1Name=base_jar
  bone2Name=pelvis
  swing1LimitDegrees=45
  swing1Motion=Free
  swing2LimitDegrees=45
  swing2Motion=Free
  twistLimitDegrees=45
  twistMotion=Free
 }
]
