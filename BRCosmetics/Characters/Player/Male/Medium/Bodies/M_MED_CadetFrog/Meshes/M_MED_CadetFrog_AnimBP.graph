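# Animation Blueprint for the CadetFrog medium male player body cosmetic, copying the attached parent mesh's pose and adding secondary motion.
# A Control Rig, with isSkydive fed by bIsParachuteOpened, precedes additive offsets on tube and backpack bones.
# Two RigidBody simulations based at pelvis and spine_05 then feed spring-limited AnimDynamics for dagger and back tube.
# The update event sets IsJumpFall (falling, not skydiving), which the AnimGraph never reads.
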AnimBlueprint /BRCosmetics/Characters/Player/Male/Medium/Bodies/M_MED_CadetFrog/Meshes/M_MED_CadetFrog_AnimBP
target_skeleton=&/M_MED_CadetFrog_Skeleton
graphs [
 AnimGraph {nodes [
   n1 = Control Rig!(Source=Copy Pose From Mesh() {bUseAttachedParent=true}, Alpha, isSkydive:bool=Get bIsParachuteOpened()) -> (Pose) {
    alphaBoolBlend {blendOption=Linear}
    controlRigAssetReference {blueprintRigClass=&/M_MED_CadetFrog_CtrlRig.M_MED_CadetFrog_CtrlRig_C}
   }
   n2 = "Transform (Modify) Bone - Bone: tube_clavicle_rt"!(ComponentPose=Local To Component(LocalPose=@n1), bAlphaBoolEnabled:bool=True, Translation="0,0,1.5", Rotation="0,0,0", Scale="1,1,1") -> (Pose) {
    alphaInputType=Bool
    translationMode=BMM_Additive
    translationSpace=BCS_BoneSpace
    alphaBoolBlend {blendInTime=0.2, blendOption=Linear, blendOutTime=0.2}
    boneToModify {boneName=tube_clavicle_rt}
    translation {z=1.5}
   }
   n3 = "Transform (Modify) Bone - Bone: attach_backpack"!(ComponentPose=@n2, bAlphaBoolEnabled:bool=True, Translation="0,-1.5,0", Rotation="0,0,0", Scale="1,1,1") -> (Pose) {
    alphaInputType=Bool
    translationMode=BMM_Additive
    translationSpace=BCS_BoneSpace
    alphaBoolBlend {blendInTime=0.2, blendOption=Linear, blendOutTime=0.2}
    boneToModify {boneName=attach_backpack}
    translation {y=-1.5}
   }
   n4 = RigidBody(ComponentPose=@n3) {
    bClampLinearTranslationLimitToRefPose=true
    bOverrideWorldGravity=true
    overridePhysicsAsset=&/M_MED_CadetFrog_Physics
    simulationSpace=BaseBoneSpace
    alphaBoolBlend {blendOption=Linear}
    baseBoneRef {boneName=pelvis}
    overrideWorldGravity {z=-980}
    simSpaceSettings {worldAlpha=1}
   }
   n5 = RigidBody(ComponentPose=@n4) {
    bClampLinearTranslationLimitToRefPose=true
    bOverrideWorldGravity=true
    overridePhysicsAsset=&/M_MED_CadetFrog_FrontTube_Physics
    simulationSpace=BaseBoneSpace
    alphaBoolBlend {blendOption=Linear}
    baseBoneRef {boneName=spine_05}
    overrideWorldGravity {z=-980}
    simSpaceSettings {worldAlpha=0.2}
   }
   n6 = Anim Dynamics - Bone: dyn_dagger(ComponentPose=@n5, Alpha=1, ExternalForce="0,0,0") {
    angularDampingOverride=1
    angularSpringConstant=20000
    bAngularSpring=true
    bOverrideAngularDamping=true
    bUseGravityOverride=true
    gravityScale=0.35
    simulationSpace=RootRelative
    alphaBoolBlend {blendOption=Linear}
    boundBone {boneName=dyn_dagger}
    componentAppliedLinearAccClamp=<6000 6000 6000>
    componentLinearAccScale=<1 1 1>
    componentLinearVelScale=<1 1 1>
    gravityOverride {z=-980}
    physicsBodyDefinitions [{
      boundBone {boneName=dyn_dagger}
      boxExtents=<40 40 30>
      constraintSetup {angularLimitsMax=<8 0 0>, angularLimitsMin=<-7 0 0>, angularTarget=<1 0 0>}
      localJointOffset=<0 0 -2>
     }]
   }
   n7 = Anim Dynamics - Bone: dyn_tube_back_cr(ComponentPose=@n6, Alpha=1, ExternalForce="0,0,0") {
    angularDampingOverride=1
    angularSpringConstant=20000
    bAngularSpring=true
    bOverrideAngularDamping=true
    bUseGravityOverride=true
    gravityScale=0.35
    simulationSpace=RootRelative
    alphaBoolBlend {blendOption=Linear}
    boundBone {boneName=dyn_tube_back_cr}
    componentAppliedLinearAccClamp=<6000 6000 6000>
    componentLinearAccScale=<1 1 1>
    componentLinearVelScale=<1 1 1>
    gravityOverride {z=-980}
    physicsBodyDefinitions [{
      boundBone {boneName=dyn_tube_back_cr}
      boxExtents=<30 30 30>
      constraintSetup {angularLimitsMax=<0 0 0>, angularLimitsMin=<0 -20 0>, angularTarget=<1 0 0>}
      localJointOffset=<6 0 0>
     }]
   }
   Output Pose(Result=Component To Local(ComponentPose=@n7)) {layerGroup=None}
  ]}
 EventGraph {nodes [
   Event BlueprintUpdateAnimation()
   n1 = Get bIsSkyDiving()
   > Set IsJumpFall(IsJumpFall:bool=AND Boolean(A=Get bIsFalling(), B=NOT Boolean(A=@n1))) -> (Output_Get:bool)
  ]}
]
variables [IsJumpFall {type=boolean}]
